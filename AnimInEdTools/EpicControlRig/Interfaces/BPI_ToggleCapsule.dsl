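; BPI_ToggleCapsule is a Blueprint Interface in the EpicControlRig folder of the AnimInEdTools content.
; It declares a single function, ToggleCapsule, with no inputs or outputs.
; The file holds only the signature, so any behavior comes from the classes that implement it.
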(fn ToggleCapsule)
